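# Material function for the OpenWorldTemplate landscape material that outputs a tiled rounded-square mask, forming grid lines between cells.
# UV, defaulting to world XY, is divided by Size and wrapped, then fed to a custom rounded-box distance function shaped by Radius and LineSize.
# Screen-space derivatives shrink the box, widen the SmoothStep edge and blend toward a flat LineSize-based value as cells get small, limiting aliasing.
# Several other branches, including a BoxMask-2D variant, never reach the output.
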MaterialFunction /Engine/OpenWorldTemplate/LandscapeMaterial/MF_RSquareGradient
graph {nodes [
  n1 = WorldPosition()
  MaterialExpressionDivide_0 = Divide(
   A=FunctionInput(Preview=ComponentMask(None=@n1.XYZ) {g=true, r=true}) {bUsePreviewValueAsDefault=true, inputName=UV, inputType=FunctionInput_Vector2}
   B=FunctionInput(Preview=Constant() {r=100}) {bUsePreviewValueAsDefault=true, inputName=Size, inputType=FunctionInput_Scalar}
  ) {constB=100}
  n2 = Custom!(x=@MaterialExpressionDivide_0) {
   code="fwidth(x);"
   outputType=CMOT_Float2
   inputs [{inputName=x, input {expression=&self:MaterialExpressionDivide_0}}]
  }
  Clamp(None=@n2) {minDefault=-1}
  Constant() {r=0.501}
  n3 = WorldPosition()
  n4 = Constant() {r=100}
  n5 = Frac(None=FunctionInput(Preview=Divide(A=@n3.XYZ, B=@n4)) {inputName=Projection, inputType=FunctionInput_Vector2})
  MaterialExpressionFunctionInput_3 = FunctionInput(Preview=Constant() {r=0.1}) {inputName=Radius, inputType=FunctionInput_Scalar}
  n6 = Subtract(A=Abs(None=ConstantBiasScale(None=@n5) {bias=-0.5, scale=2}), B=OneMinus(None=@MaterialExpressionFunctionInput_3))
  Distance(A=Saturate(None=Divide(A=@n6, B=@MaterialExpressionFunctionInput_3)), B=Constant())
  n7 = FunctionInput(Preview=ScalarParameter() {defaultValue=0.12, parameterName=Size, sliderMax=1}) {inputName=LineSize, inputType=FunctionInput_Scalar}
  n8 = Saturate(None=Subtract(A=MaterialFunctionCall(Vector 2=@n2) {materialFunction=/Engine/Functions/Engine_MaterialFunctions02/Utility/VectorLength}.V2 Length, B=@n7))
  MaterialExpressionComponentMask_3 = ComponentMask(None=ConstantBiasScale(None=Frac(None=@MaterialExpressionDivide_0)) {bias=-0.5, scale=2}) {g=true, r=true}
  n9 = OneMinus(None=Add(A=@MaterialExpressionFunctionInput_3, B=@n7))
  MaterialExpressionSubtract_1 = Subtract(A=@n9, B=@n2)
  n10 = Custom!(r=@MaterialExpressionFunctionInput_3, p=@MaterialExpressionComponentMask_3, b=@MaterialExpressionSubtract_1) {
   code="  float2 q = abs(p) - b;\r\n  return length(max(q,0.0)) + min(max(q.x, q.y),0.0) - r;"
   inputs [
    {inputName=r, input {expression=&self:MaterialExpressionFunctionInput_3}}
    {inputName=p, input {expression=&self:MaterialExpressionComponentMask_3}}
    {inputName=b, input {expression=&self:MaterialExpressionSubtract_1}}
   ]
  }
  n11 = ScalarParameter() {defaultValue=1}
  FunctionOutput(None=LinearInterpolate(A=SmoothStep(Max=@n8, Value=@n10), B=Power(Base=@n7, Exp=@n11), Alpha=@n8))
  MaterialExpressionComponentMask_1 = ComponentMask(None=@MaterialExpressionDivide_0) {g=true, r=true}
  n12 = Custom!(x=@MaterialExpressionComponentMask_1) {
   code="fwidth(x);"
   outputType=CMOT_Float2
   inputs [{inputName=x, input {expression=&self:MaterialExpressionComponentMask_1}}]
  }
  n13 = Saturate(None=@n12)
  n14 = MaterialFunctionCall(Vector 2=@n13) {materialFunction=&VectorLength}
  n15 = OneMinus(None=@n14.V2 Length)
  n16 = Multiply(A=@n15) {constB=0.5}
  n17 = Add(A=Multiply(A=@n14.V2 Length) {constB=0.5}) {constB=0.5}
  n18 = Subtract(A=OneMinus(None=@n7), B=@MaterialExpressionFunctionInput_3)
  n19 = MaterialFunctionCall!(
   A=Frac(None=@MaterialExpressionDivide_0)
   B=Constant() {r=0.5}
   Bounds=Multiply(A=@n18, B=OneMinus(None=@n13))
   Edge Falloff=@MaterialExpressionFunctionInput_3
  ) {materialFunction=/Engine/Functions/Engine_MaterialFunctions01/Gradient/BoxMask-2D}
  LinearInterpolate(A=OneMinus(None=SmoothStep(Min=@n16, Max=@n17, Value=@n19.Result) {constMax=0.51}), B=OneMinus(None=@n7), Alpha=@n14.V2 Length)
  Multiply(A=@n18, B=@n15)
  Multiply(A=@n2, B=Multiply(A=@n9, B=OneMinus(None=@n8)))
  ScalarParameter() {defaultValue=0.99, parameterName=Param_1}
 ]}
