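# Procedural grid and checker landscape material for the Open World template, set up with material attributes and Nanite usage.
# MF_ProcGrid supplies wire and checker masks, with tri-planar or uni-planar projection, driving base color, roughness and metallic for lines and squares.
# Optional static switches add slope darkening from world normal Z and an overlay water tint below a faded sea level from world height.
# Named reroutes carry the masks.
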Material /Engine/OpenWorldTemplate/LandscapeMaterial/M_ProcGrid
graph {
 properties {bUseMaterialAttributes=true, bUsedWithNanite=true, pixelDepthOffsetMode=PDOM_Legacy, usedShadingModels=Default Lit}
 nodes [
  n1 = MaterialFunctionCall!(
   TriPlanar=StaticBoolParameter() {group=Grid Properties, parameterName="Tri-Planar (Or Uni-Planar)"}
   TileSize=ScalarParameter() {defaultValue=100, group=Grid Properties, parameterName=Tile Scale}
  ) {materialFunction=&/MF_ProcGrid}
  NamedRerouteDeclaration(None=OneMinus(None=@n1.Wire)) {name=Line Mask, nodeColor=<rgba 0.8281 0 1 1>}
  NamedRerouteDeclaration(None=@n1.Checker) {name=Checker Mask, nodeColor=<rgba 1 0.7031 0 1>}
  n2 = VertexNormalWS()
  n3 = VectorParameter() {desc="R = Strength \r\nG = Power\r\nB = Offset", group=Side Tint, parameterName=Side Parameters, defaultValue=<rgba 1.0402 28.1525 0.3 0>}
  n4 = Power(Base=Multiply(A=ComponentMask(None=@n2) {b=true}, B=@n3.R), Exp=@n3.G)
  NamedRerouteDeclaration(None=Saturate(None=Add(A=@n4, B=@n3.B))) {name=Angle Darken, nodeColor=<rgba 0 0.4732 1 1>}
  n5 = Divide(
   A=Add(A=WorldPosition().XYZ, B=ScalarParameter() {defaultValue=1203.7327, group=Water Level, parameterName=Sea Level Base})
   B=ScalarParameter() {defaultValue=265.9673, group=Water Level, parameterName=Fade}
  )
  NamedRerouteDeclaration(None=Saturate(None=ComponentMask(None=@n5) {b=true})) {name=Water Mask, nodeColor=<rgba 0 1 0.9227 1>}
  n6 = LinearInterpolate(
   A=VectorParameter() {group=Checker, parameterName=Checker Colour 1, defaultValue=<rgba 0.2917 0.2917 0.2917 0>}.RGB
   B=VectorParameter() {group=Checker, parameterName=Checker Colour 2, defaultValue=<rgba 0.2396 0.2396 0.2396 0>}.RGB
   Alpha=Multiply(A=NamedRerouteUsage())
  )
  n7 = StaticSwitchParameter(True=Multiply(A=@n6, B=NamedRerouteUsage()), False=@n6) {defaultValue=true, group=Side Tint, parameterName=Enable Side Tint}
  n8 = LinearInterpolate(
   A=MaterialFunctionCall(Base=VectorParameter() {group=Water Level, parameterName=Water Colour, defaultValue=<rgba 0.1868 0.49 0.7083 0>}.RGB, Blend=@n7) {materialFunction=&Blend_Overlay}.Result
   B=@n7
   Alpha=NamedRerouteUsage()
  )
  n9 = ScalarParameter() {defaultValue=0.5, group=Checker, parameterName=Checker Rough 1, sliderMin=0.2}
  n10 = ScalarParameter() {defaultValue=0.65, group=Checker, parameterName=Checker Rough 2}
  n11 = NamedRerouteUsage()
  n12 = NamedRerouteUsage()
  SetMaterialAttributes!(
   MaterialAttributes
   Base Color=LinearInterpolate(
    A=VectorParameter() {group=Checker, parameterName=Line Colour, defaultValue=<rgba 0.3385 0.3385 0.3385 0>}.RGB
    B=StaticSwitchParameter(True=@n8, False=@n7) {defaultValue=true, group=Water Level, parameterName=Water Level Tint, sortPriority=0}
    Alpha=NamedRerouteUsage()
   )
   Roughness=LinearInterpolate(A=Constant() {r=0.3}, B=LinearInterpolate(A=@n9, B=@n10, Alpha=@n11), Alpha=@n12)
   Metallic=OneMinus(None=@n12)
  ) {attributeSetTypes [#1, #2, #3]}
 ]
}
